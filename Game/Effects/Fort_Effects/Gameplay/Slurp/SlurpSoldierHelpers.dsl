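; Slurp Soldier helper functions that return a player's shield as a fraction of maximum, starting from either a skeletal mesh component or a player pawn.
; Both share the same logic, going through the mesh's Fort player anim instance to its pawn's player state for shield values, with the divisor clamped to at least one.
; If the anim instance cast fails, they return the FrontEnd_Idle_Bypass curve value instead; an invalid input returns zero.
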(fn GetShieldPercentageFromOwnerMeshComponent
  (params (Owner :type Skeletal%20Mesh%20Component%20Object%20Reference))
  (bind _returnvalue (Components|SkeletalMesh|GetAnimInstance Owner))
  (Utilities|IsValid Owner
    (:Is%20Valid
      (bind _asfort_player_anim_instance (Utilities|Casting|CastToFortPlayerAnimInstance _returnvalue)
        (:then
          (Utilities|IsValid (Variables|/Script/FortniteGame.FortPlayerAnimInstance|GetFortPlayerPawn _asfort_player_anim_instance)
            (:Is%20Valid
              (bind _asfort_player_state_zone (Utilities|Casting|CastToFortPlayerStateZone (Variables|/Script/Engine.Pawn|GetPlayerState (Variables|/Script/FortniteGame.FortPlayerAnimInstance|GetFortPlayerPawn _asfort_player_anim_instance))))
              (bind (_health _healthmax _shield _shieldmax) (HealthSet|GetHealthandShields _asfort_player_state_zone))
              (return (/ _shield (Math|Float|Max(Float) _shieldmax 1.0))))))
        (:CastFailed
          (return (Animation|Curves|GetCurveValue _returnvalue "FrontEnd_Idle_Bypass")))))
    (:Is%20Not%20Valid
      (return 0.0))))

(fn GetShieldPercentageFromPawn
  (params (Pawn :type Fort%20Player%20Pawn%20Object%20Reference))
  (bind _mesh (Variables|/Script/Engine.Character|GetMesh Pawn))
  (bind _returnvalue (Components|SkeletalMesh|GetAnimInstance _mesh))
  (Utilities|IsValid Pawn
    (:Is%20Valid
      (bind _asfort_player_anim_instance (Utilities|Casting|CastToFortPlayerAnimInstance _returnvalue)
        (:then
          (Utilities|IsValid (Variables|/Script/FortniteGame.FortPlayerAnimInstance|GetFortPlayerPawn _asfort_player_anim_instance)
            (:Is%20Valid
              (bind _asfort_player_state_zone (Utilities|Casting|CastToFortPlayerStateZone (Variables|/Script/Engine.Pawn|GetPlayerState (Variables|/Script/FortniteGame.FortPlayerAnimInstance|GetFortPlayerPawn _asfort_player_anim_instance))))
              (bind (_health _healthmax _shield _shieldmax) (HealthSet|GetHealthandShields _asfort_player_state_zone))
              (return (/ _shield (Math|Float|Max(Float) _shieldmax 1.0))))))
        (:CastFailed
          (return (Animation|Curves|GetCurveValue _returnvalue "FrontEnd_Idle_Bypass")))))
    (:Is%20Not%20Valid
      (return 0.0))))
